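# Physics asset for the Glare Male pickaxe mesh, defining secondary-motion bodies for a back strap and four wheels, all tied to b_AR_Root.
# The strap is a three-segment capsule chain with narrow swing and zero twist limits, hung from a master body that rotates freely on the root.
# Each wheel is a zero-length capsule, effectively a sphere, linked through its master body to the root by hinge-like joints free only in swing2.
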PhysicsAsset /BRCosmetics/Weapons/FORT_Melee/Pickaxe_Glare_Male/Meshes/Glare_Male_Axe_PhysicsAsset
bodies {
 b_AR_Root [{shapeType=Capsule, length=0.505, radius=0.505, shapeName=b_AR_Root_capsule, rotation=<rot 0 -90 0>}]
 dyn_backStrap_01 [{shapeType=Capsule, length=2, radius=2, shapeName=dyn_backStrap_01_capsule, center=<2.5 -0.0652 -0.0916>, rotation=<rot 0 0.0001 -90>}]
 dyn_backStrap_02 [{shapeType=Capsule, length=2, radius=2, shapeName=dyn_backStrap_02_capsule, center=<2.5 -0.0652 0.7552>, rotation=<rot 0 0.0001 -89.9999>}]
 dyn_backStrap_03 [{shapeType=Capsule, length=2, radius=2, shapeName=dyn_backStrap_03_capsule, center=<2.5 -0.0652 -0.0366>, rotation=<rot 0 0 -88.0118>}]
 dyn_backStrap_MSTR [{shapeType=Capsule, length=0.505, radius=0.505, shapeName=dyn_backStrap_MSTR_capsule, rotation=<rot 0 -90 0>}]
 dyn_wheel01_01 [{shapeType=Capsule, length=0, radius=6.2, shapeName=dyn_wheel03_01_capsule, center=<5 0 0>, rotation=<rot 0 -90 -0.0001>}]
 dyn_wheel01_MSTR [{shapeType=Capsule, length=0.505, radius=0.505, shapeName=dyn_wheel01_MSTR_capsule, rotation=<rot 0 -90 0>}]
 dyn_wheel02_01 [{shapeType=Capsule, length=0, radius=6.1, shapeName=dyn_wheel03_01_capsule, center=<5 0 0>, rotation=<rot 0 -90 0>}]
 dyn_wheel02_MSTR [{shapeType=Capsule, length=0.505, radius=0.505, shapeName=dyn_wheel02_MSTR_capsule, rotation=<rot 0 -90 0>}]
 dyn_wheel03_01 [{shapeType=Capsule, length=0, radius=6.5, shapeName=dyn_wheel03_01_capsule, center=<5 0 0>, rotation=<rot 0 -90 0>}]
 dyn_wheel03_MSTR [{shapeType=Capsule, length=0.505, radius=0.505, shapeName=dyn_wheel03_MSTR_capsule, rotation=<rot 0 -90 0>}]
 dyn_wheel04_01 [{shapeType=Capsule, length=0, radius=6.3, shapeName=dyn_wheel03_01_capsule, center=<5 0 0>, rotation=<rot 0 -90 0>}]
 dyn_wheel04_MSTR [{shapeType=Capsule, length=0.505, radius=0.505, shapeName=dyn_wheel04_MSTR_capsule, rotation=<rot 0 -90 0>}]
}
constraints [
 {bone1Name=dyn_backStrap_01, bone2Name=dyn_backStrap_MSTR, swing1LimitDegrees=20, swing2LimitDegrees=15, twistLimitDegrees=0}
 {bone1Name=dyn_backStrap_02, bone2Name=dyn_backStrap_01, swing1LimitDegrees=20, swing2LimitDegrees=10, twistLimitDegrees=0}
 {bone1Name=dyn_backStrap_03, bone2Name=dyn_backStrap_02, swing1LimitDegrees=20, swing2LimitDegrees=10, twistLimitDegrees=0}
 {bone1Name=dyn_wheel01_01, bone2Name=dyn_wheel01_MSTR, swing1LimitDegrees=0, swing2LimitDegrees=45, swing2Motion=Free, twistLimitDegrees=0}
 {bone1Name=dyn_wheel02_01, bone2Name=dyn_wheel02_MSTR, swing1LimitDegrees=0, swing2LimitDegrees=45, swing2Motion=Free, twistLimitDegrees=0}
 {bone1Name=dyn_wheel03_01, bone2Name=dyn_wheel03_MSTR, swing1LimitDegrees=0, swing2LimitDegrees=45, swing2Motion=Free, twistLimitDegrees=0}
 {bone1Name=dyn_wheel04_01, bone2Name=dyn_wheel04_MSTR, swing1LimitDegrees=0, swing2LimitDegrees=45, swing2Motion=Free, twistLimitDegrees=0}
 {
  bone1Name=dyn_backStrap_MSTR
  bone2Name=b_AR_Root
  swing1LimitDegrees=45
  swing1Motion=Free
  swing2LimitDegrees=45
  swing2Motion=Free
  twistLimitDegrees=45
  twistMotion=Free
 }
 {bone1Name=dyn_wheel01_MSTR, bone2Name=b_AR_Root, swing1LimitDegrees=0, swing2LimitDegrees=45, swing2Motion=Free, twistLimitDegrees=0}
 {bone1Name=dyn_wheel02_MSTR, bone2Name=b_AR_Root, swing1LimitDegrees=0, swing2LimitDegrees=45, swing2Motion=Free, twistLimitDegrees=0}
 {bone1Name=dyn_wheel03_MSTR, bone2Name=b_AR_Root, swing1LimitDegrees=0, swing2LimitDegrees=45, swing2Motion=Free, twistLimitDegrees=0}
 {bone1Name=dyn_wheel04_MSTR, bone2Name=b_AR_Root, swing1LimitDegrees=0, swing2LimitDegrees=45, swing2Motion=Free, twistLimitDegrees=0}
]
