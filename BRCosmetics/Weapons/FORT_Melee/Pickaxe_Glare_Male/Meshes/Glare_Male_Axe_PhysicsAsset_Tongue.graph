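# Physics asset for the tongue of the Glare Male pickaxe cosmetic, defining a short jointed chain for the axe mesh's dynamic tongue bones.
# Small capsules sit on the axe root and tongue master bone, with larger offset capsules on the two tongue segments.
# Both segment joints allow narrow, asymmetric swing with twist limited to zero, while the master-to-root constraint leaves swing and twist free.
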PhysicsAsset /BRCosmetics/Weapons/FORT_Melee/Pickaxe_Glare_Male/Meshes/Glare_Male_Axe_PhysicsAsset_Tongue
bodies {
 b_AR_Root [{shapeType=Capsule, length=0.505, radius=0.505, shapeName=b_AR_Root_capsule, rotation=<rot 0 -90 0>}]
 dyn_tongue_01 [{shapeType=Capsule, length=2, radius=2, shapeName=dyn_tongue_01_capsule, center=<2.5 0 0>, rotation=<rot 90 -90 0>}]
 dyn_tongue_02 [{shapeType=Capsule, length=2, radius=2, shapeName=dyn_tongue_02_capsule, center=<2.5 0 0>, rotation=<rot 90 -90 0>}]
 dyn_tongue_MSTR [{shapeType=Capsule, length=0.505, radius=0.505, shapeName=dyn_tongue_MSTR_capsule, rotation=<rot 0 -90 0>}]
}
constraints [
 {bone1Name=dyn_tongue_01, bone2Name=dyn_tongue_MSTR, swing1LimitDegrees=5, swing2LimitDegrees=25, twistLimitDegrees=0}
 {bone1Name=dyn_tongue_02, bone2Name=dyn_tongue_01, swing1LimitDegrees=5, swing2LimitDegrees=25, twistLimitDegrees=0}
 {
  bone1Name=dyn_tongue_MSTR
  bone2Name=b_AR_Root
  swing1LimitDegrees=45
  swing1Motion=Free
  swing2LimitDegrees=45
  swing2Motion=Free
  twistLimitDegrees=45
  twistMotion=Free
 }
]
